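; Looping gameplay cue that plays sprint camera shakes for a player while riding.
; On loop start it resolves the target pawn's player controller and camera manager, then starts an FOV flare shake and a half-scale looping shake in camera-local space.
; It stores the controller and both shake instances, and stops both shakes immediately when the cue ceases to be relevant.
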(event EventOnLoopingStart(EitherVFX-preferthisversion)
  (params (MyTarget :type Actor%20Object%20Reference) (Parameters :type Gameplay%20Cue%20Parameters%20Structure%20%28by%20ref%29) (ParticleComponents :type Array%20of%20FXSystem%20Component%20Object%20References) (AudioComponents :type Array%20of%20Audio%20Component%20Object%20References))
  (bind _asfort_player_pawn (Utilities|Casting|CastToFortPlayerPawn MyTarget))
  (Utilities|IsValid _asfort_player_pawn
    (:Is%20Valid
      (Utilities|IsValid (Fortnite|GetFortPlayerControllerfromActor _asfort_player_pawn)
        (:Is%20Valid
          (bind _output_get (Variables|Default|SetFortPlayerController (Fortnite|GetFortPlayerControllerfromActor _asfort_player_pawn)))
          (Utilities|IsValid (Variables|/Script/Engine.PlayerController|GetPlayerCameraManager _output_get)
            (:Is%20Valid
              (bind _returnvalue (CameraShakes|StartCameraShake :self (Variables|/Script/Engine.PlayerController|GetPlayerCameraManager _output_get) :ShakeClass "/RidingContent/Gameplay/Camera/CameraShakes/CameraShake_Riding_Sprint_FOVFlare.CameraShake_Riding_Sprint_FOVFlare_C" :Scale 1.0 :PlaySpace "CameraLocal"))
              (Variables|Default|SetFOVFlare _returnvalue)
              (bind _returnvalue_2 (CameraShakes|StartCameraShake :self (Variables|/Script/Engine.PlayerController|GetPlayerCameraManager _output_get) :ShakeClass "/RidingContent/Gameplay/Camera/CameraShakes/CameraShake_Riding_Sprint_Loop.CameraShake_Riding_Sprint_Loop_C" :Scale 0.5 :PlaySpace "CameraLocal"))
              (Variables|Default|SetShakeNoise _returnvalue_2))))))))

(fn OnCeaseRelevant
  (params (MyTarget :type Actor%20Object%20Reference) (Parameters :type Gameplay%20Cue%20Parameters%20Structure%20%28by%20ref%29))
  (bind _fortplayercontroller (Variables|Default|GetFortPlayerController))
  (bind _playercameramanager (Variables|/Script/Engine.PlayerController|GetPlayerCameraManager _fortplayercontroller))
  (Utilities|IsValid _fortplayercontroller
    (:Is%20Valid
      (Utilities|IsValid _playercameramanager
        (:Is%20Valid
          (CameraShakes|StopCameraShake _playercameramanager (Variables|Default|GetFOVFlare) true)
          (CameraShakes|StopCameraShake _playercameramanager (Variables|Default|GetShakeNoise) true)
          (return false))))))

(fn ConstructionScript)
