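; Latent gameplay cue for Burt charging, from the riding content, that adds a speed-line camera lens effect to the player during a boost.
; On execute it caches the target's Fort pawn and controller, then runs the boost logic.
; That adds the looping lens effect to the camera manager with its streak spawn rate set to zero.
; A fixed three-second timer ends and recycles the cue, without explicitly removing the lens effect.
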(event Custom|ExecuteBoostLogic
  (bind _gcplayercontroller (Variables|Default|GetGCPlayerController))
  (bind _playercameramanager (Variables|/Script/Engine.PlayerController|GetPlayerCameraManager _gcplayercontroller))
  ;[1 1 1] Boost Start
  (Utilities|Time|SetTimerbyEvent (EventDispatchers|CreateEvent :!event Custom|BoostEndEvent) 3.0)
  ;;[1 1 1] SpeedLine Lens FX
  (Utilities|IsValid _playercameramanager
    (:Is%20Valid
      (bind _returnvalue (Camera|AddGenericCameraLensEffect _playercameramanager "/Game/Effects/Fort_Effects/Camera_Effects/BP_SpeedLines_Looping_Camera_LensV2.BP_SpeedLines_Looping_Camera_LensV2_C"))
      (Effects|Components|ParticleSystem|SetFloatParameter (CameraLensEffect|GetPrimaryParticleComponent _returnvalue) "StreaksSpawnRate" 0.0))))

(event Custom|BoostEndEvent
  ;[1 1 1] Boost End
  (GameplayCueNotify|End%28Recycle%29GameplayCue))

(fn OnExecute
  (params (MyTarget :type Actor%20Object%20Reference) (Parameters :type Gameplay%20Cue%20Parameters%20Structure%20%28by%20ref%29))
  (bind _asfort_player_pawn (Utilities|Casting|CastToFortPlayerPawn MyTarget))
  (bind _output_get (Variables|Default|SetFortPlayerPawn _asfort_player_pawn))
  (Variables|Default|SetGCPlayerController (Fortnite|GetFortPlayerControllerfromActor _output_get))
  (CallFunction|ExecuteBoostLogic)
  (return false))

(fn ConstructionScript)
